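; Blueprint that animates the Cattus POI overgrowth by driving Cattus and Sword height-blend parameters in MPC_CattusPOI.
; BeginPlay watches calendar event "PCVD" and, within its first ten minutes, starts a self-rescheduling timer that turns elapsed time into a normalized value.
; That value's OnRep has remote instances lerp the blend parameters.
; A timeline debug event and a Construction Script debug value drive the same parameters for previewing.
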(event Custom|UpdateOvergrowth
  (bind (_biseventactive _timesincebegin) (CalendarEvents|GetCalendarEventInformation "PCVD"))
  (if _biseventactive
    (bind _output_get (Variables|Default|SetSecondsSinceBeginning (Math|Timespan|GetTotalSeconds _timesincebegin)))
    (if (< _output_get 6e+02)
      (Variables|Default|SetNormalizedValue (Math|Float|Clamp(Float) (/ (Variables|Default|GetSecondsSinceBeginning) 6e+02)))
      (Utilities|Time|SetTimerbyEvent (EventDispatchers|CreateEvent :!event Custom|UpdateOvergrowth) (Variables|Default|GetTimerTime))
      (else
        (Variables|Default|SetNormalizedValue 1.0)))))

(event Custom|ClientUpdate
  (bind _normalizedvalue (Variables|Default|GetNormalizedValue))
  (Rendering|Material|SetScalarParameterValue "/Game/Athena/Environments/Cattus/Materials/MPC_CattusPOI.MPC_CattusPOI" "CattusHeightBlend_Distance" (Math|Float|Lerp (Variables|Default|GetCattusBlend_Distance) 5.5e+03 _normalizedvalue) :!class /Script/Engine.KismetMaterialLibrary)
  (Rendering|Material|SetScalarParameterValue "/Game/Athena/Environments/Cattus/Materials/MPC_CattusPOI.MPC_CattusPOI" "CattusHeightBlend_Offset" (Math|Float|Lerp (Variables|Default|GetCattusBlend_Offset) 456.0 _normalizedvalue) :!class /Script/Engine.KismetMaterialLibrary)
  (Rendering|Material|SetScalarParameterValue "/Game/Athena/Environments/Cattus/Materials/MPC_CattusPOI.MPC_CattusPOI" "SwordHeightBlend_Distance" (Math|Float|Lerp (Variables|Default|GetSwordBlend_Distance) 8e+03 (Variables|Default|GetNormalizedValue)) :!class /Script/Engine.KismetMaterialLibrary)
  (Rendering|Material|SetScalarParameterValue "/Game/Athena/Environments/Cattus/Materials/MPC_CattusPOI.MPC_CattusPOI" "SwordHeightBlend_Offset" (Math|Float|Lerp (Variables|Default|GetSwordBlend_Offset) 2e+02 (Variables|Default|GetNormalizedValue)) :!class /Script/Engine.KismetMaterialLibrary))

(event Custom|DebugOvergrowth
  (params (NewParam :type Float%20%28double-precision%29 :default 2e+01))
  (Components|Timeline|SetPlayRate (Variables|BP_Cattus_MPC_Animation|GetTimeline_0) NewParam)
  (bind (_direction _normalizedgrowth) (Timeline|Timeline_0 PlayFromStart :NewTime (Math|Float|Clamp(Float) (Math|Timespan|GetTotalSeconds (stale EventBeginPlay _timesincebegin)) 0.0 6e+02))
    (:Update
      (Rendering|Material|SetScalarParameterValue "/Game/Athena/Environments/Cattus/Materials/MPC_CattusPOI.MPC_CattusPOI" "CattusHeightBlend_Distance" (Math|Float|Lerp (Variables|Default|GetCattusBlend_Distance) 5.5e+03 _normalizedgrowth) :!class /Script/Engine.KismetMaterialLibrary)
      (Rendering|Material|SetScalarParameterValue "/Game/Athena/Environments/Cattus/Materials/MPC_CattusPOI.MPC_CattusPOI" "CattusHeightBlend_Offset" (Math|Float|Lerp (Variables|Default|GetCattusBlend_Offset) 456.0 _normalizedgrowth) :!class /Script/Engine.KismetMaterialLibrary)
      (Rendering|Material|SetScalarParameterValue "/Game/Athena/Environments/Cattus/Materials/MPC_CattusPOI.MPC_CattusPOI" "SwordHeightBlend_Distance" (Math|Float|Lerp (Variables|Default|GetSwordBlend_Distance) 8e+03 _normalizedgrowth) :!class /Script/Engine.KismetMaterialLibrary)
      (Rendering|Material|SetScalarParameterValue "/Game/Athena/Environments/Cattus/Materials/MPC_CattusPOI.MPC_CattusPOI" "SwordHeightBlend_Offset" (Math|Float|Lerp (Variables|Default|GetSwordBlend_Offset) 2e+02 _normalizedgrowth) :!class /Script/Engine.KismetMaterialLibrary))))

(event EventBeginPlay
  (bind (_eventname _timeuntilend _timesincebegin) (MCPContext|WatchCalendarEvent "PCVD")
    (:OnEventActive
      (if (< (Math|Timespan|GetTotalSeconds _timesincebegin) 6e+02)
        (CallFunction|UpdateOvergrowth)))))

(fn OnRep_NormalizedValue
  (Networking|SwitchHasAuthority
    (:Remote
      (CallFunction|ClientUpdate))))

(fn ConstructionScript
  (bind _debugvalue (Variables|Default|GetDebugValue))
  (if (Variables|Default|GetDebug)
    (Rendering|Material|SetScalarParameterValue "/Game/Athena/Environments/Cattus/Materials/MPC_CattusPOI.MPC_CattusPOI" "CattusHeightBlend_Distance" (Math|Float|Lerp (Variables|Default|GetCattusBlend_Distance) 5.5e+03 _debugvalue) :!class /Script/Engine.KismetMaterialLibrary)
    (Rendering|Material|SetScalarParameterValue "/Game/Athena/Environments/Cattus/Materials/MPC_CattusPOI.MPC_CattusPOI" "CattusHeightBlend_Offset" (Math|Float|Lerp (Variables|Default|GetCattusBlend_Offset) 456.0 _debugvalue) :!class /Script/Engine.KismetMaterialLibrary)
    (Rendering|Material|SetScalarParameterValue "/Game/Athena/Environments/Cattus/Materials/MPC_CattusPOI.MPC_CattusPOI" "SwordHeightBlend_Distance" (Math|Float|Lerp (Variables|Default|GetSwordBlend_Distance) 8e+03 _debugvalue) :!class /Script/Engine.KismetMaterialLibrary)
    (Rendering|Material|SetScalarParameterValue "/Game/Athena/Environments/Cattus/Materials/MPC_CattusPOI.MPC_CattusPOI" "SwordHeightBlend_Offset" (Math|Float|Lerp (Variables|Default|GetSwordBlend_Offset) 2e+02 _debugvalue) :!class /Script/Engine.KismetMaterialLibrary)))
